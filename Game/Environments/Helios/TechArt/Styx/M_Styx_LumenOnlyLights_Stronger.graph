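# Unlit, Nanite-enabled material for Styx in the Helios environment, named as a stronger Lumen-only light that adds emissive glow to ray-traced lighting.
# A ray tracing quality switch supplies black normally and the Emissive color parameter when ray traced.
# A height mask multiplies that color, keeping it above the object's origin and fading it out just below.
# Sphere radius, Bias, camera-height and Styx_GI_Lighting nodes appear as unused side branches.
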Material /Game/Environments/Helios/TechArt/Styx/M_Styx_LumenOnlyLights_Stronger
graph {
 properties {bUsedWithNanite=true, pixelDepthOffsetMode=PDOM_Legacy, shadingModel=MSM_Unlit, usedShadingModels=Unlit}
 nodes [
  n1 = Constant3Vector() {constant=<rgba 0 0 1 0>}
  Multiply(B=Multiply(
    A=Transform(None=@n1) {transformSourceType=TRANSFORMSOURCE_World, transformType=TRANSFORM_Local}
    B=ScalarParameter() {parameterName=Sphere radius}
   ))
  n2 = Subtract(A=ComponentMask(None=ObjectPositionWS()) {b=true}, B=WorldPosition().Z)
  Multiply(
   A=RayTracingQualitySwitch(Normal=Constant(), RayTraced=VectorParameter() {parameterName=Emissive, defaultValue=<rgba 12 0.0055 0 1>}.RGB)
   B=OneMinus(None=Saturate(None=@n2))
  )
  ScalarParameter() {parameterName=Bias}
  Subtract(A=@n2, B=ComponentMask(None=CameraPositionWS()) {b=true})
  VectorParameter() {parameterName=Styx_GI_Lighting, defaultValue=<rgba 3.8491 20 3.9038 1>}
 ]
}
